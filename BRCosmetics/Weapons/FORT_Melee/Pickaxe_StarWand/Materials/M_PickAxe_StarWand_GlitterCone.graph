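# Translucent, unlit, two-sided effect material for the Star Wand pickaxe's glitter cone.
# Three panning, noise-distorted smoke layers drive a three-color emissive gradient and an opacity shaped by Fresnel, gradient masks, depth fade and time of day.
# World position offset adds masked, panning distortion, replaced by a constant on Low quality and mobile, to an axis rotation that varies along V, with normals corrected for the Fresnel.
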Material /BRCosmetics/Weapons/FORT_Melee/Pickaxe_StarWand/Materials/M_PickAxe_StarWand_GlitterCone
graph {
 properties {
  bUsedWithMeshDeformer=true
  bUsedWithSkeletalMesh=true
  blendMode=BLEND_Translucent
  pixelDepthOffsetMode=PDOM_Legacy
  shadingModel=MSM_Unlit
  twoSided=true
  usedShadingModels=Unlit
 }
 nodes [
  n1 = AppendVector(A=Constant2Vector(), B=ScalarParameter() {group=WPO, parameterName=BaseHeight})
  n2 = VectorParameter() {group=WPO, parameterName=TextureWPO_UVtiling, defaultValue=<rgba 1 0.25 0 0>}
  n3 = Panner(Coordinate=Multiply(A=TextureCoordinate(), B=ComponentMask(None=@n2.RGB) {g=true, r=true})) {speedY=0.25}
  n4 = TextureSample(UVs=@n3) {samplerType=SAMPLERTYPE_Normal, texture=&T_Weap_Distort_N_02}
  n5 = AppendVector(
   A=Multiply(A=ComponentMask(None=@n4.RGB) {g=true, r=true}) {constB=0.5}
   B=Multiply(
    A=TextureSample(UVs=@n3) {texture=&T_Smooth_tile}.R
    B=ScalarParameter() {defaultValue=0.25, group=WPO, parameterName=WPO_Height}
   ) {constB=0.6}
  )
  n6 = Constant() {r=0.5}
  n7 = SphereMask(
   A=Power(
    Base=MaterialFunctionCall() {materialFunction=&LinearGradient}.VGradient
    Exp=ScalarParameter() {defaultValue=0.4, group=WPO, parameterName=CenterOffset}
   ) {constExponent=0.5}
   B=@n6
   Radius=@n6
  ) {hardnessPercent=0}
  n8 = Multiply(A=Transform(None=LinearInterpolate(A=@n1, B=@n5, Alpha=@n7)), B=ScalarParameter() {defaultValue=80, parameterName=WPO_Strength}) {constB=80}
  n9 = Constant3Vector() {desc=Disable WPO anim in Low, constant=<rgba 0 0 1 0>}
  n10 = ShadingPathSwitch(Default=@n8, Deferred=@n8, Mobile=@n9) {desc=Migrated from FeatureLevelSwitch}
  n11 = Normalize(VectorInput=VectorParameter() {parameterName=RotationAxis, defaultValue=<rgba 0 1 0 1>}.RGB)
  n12 = TextureCoordinate()
  n13 = Multiply(A=OneMinus(None=ComponentMask(None=@n12) {g=true}), B=ScalarParameter() {defaultValue=-0.1, parameterName=RotationAngle})
  n14 = Add(
   A=MaterialFunctionCall() {materialFunction=&ObjectPivotPoint}.Object Pivot Location
   B=Transform(None=VectorParameter() {parameterName=Add PivotPoint, defaultValue=<rgba 0 0 25 1>}.RGB)
  )
  Add(
   A=QualitySwitch(Default=@n10, Low=@n9, High=@n10, Medium=@n10)
   B=RotateAboutAxis(NormalizedRotationAxis=@n11, RotationAngle=@n13, PivotPoint=@n14, Position=WorldPosition().XYZ)
  )
  Constant() {r=1}
  n15 = TextureCoordinate()
  n16 = TextureSample(UVs=Panner(Coordinate=@n15) {speedY=0.1}) {texture=&T_NoisePack_01}
  n17 = Multiply(
   A=Divide(A=ComponentMask(None=@n16.RGB) {g=true, r=true}) {constB=15}
   B=ScalarParameter() {defaultValue=1, parameterName=UV_Distort_Multiply}
  )
  n18 = Add(A=Panner(Coordinate=@n15) {speedY=0.15}, B=@n17)
  n19 = TextureCoordinate()
  n20 = ComponentMask(None=VectorParameter() {parameterName=TextureB_UVtiling, defaultValue=<rgba 5 0.5 0 0>}.RGB) {g=true, r=true}
  n21 = Add(A=@n17, B=Panner(Coordinate=Multiply(A=@n19, B=@n20)) {speedY=0.25})
  n22 = TextureCoordinate()
  n23 = ComponentMask(None=VectorParameter() {parameterName=TextureC_UVtiling, defaultValue=<rgba 3 0.25 0 0>}.RGB) {g=true, r=true}
  n24 = Add(A=@n17, B=Panner(Coordinate=Multiply(A=@n22, B=@n23)) {speedY=0.2})
  n25 = Power(Base=MaterialFunctionCall(
    A=TextureSample(UVs=@n18) {desc=A, samplerType=SAMPLERTYPE_LinearColor, texture=/Game/Global/Textures/Default/Tileable_Patterns/T_Tile_Smoke_Small}.R
    B=TextureSample(UVs=@n21) {desc=B, samplerType=SAMPLERTYPE_LinearColor, texture=&T_Tile_Smoke_Small}.R
    C=TextureSample(UVs=@n24) {desc=C, texture=/Game/Global/Textures/Default/Tileable_Patterns/T_Smoke_Puff_Tile_Up}.R
    Alpha=MaterialFunctionCall() {materialFunction=&LinearGradient}.VGradient
   ) {materialFunction=&3ColorBlend}.Result) {constExponent=1}
  n26 = ComponentMask(None=TextureCoordinate()) {g=true}
  n27 = Multiply(A=@n25, B=@n26)
  n28 = ScalarParameter() {defaultValue=1, parameterName=Exp}
  MaterialFunctionCall(Emissive=MaterialFunctionCall(
    A=VectorParameter() {parameterName=Color 3, defaultValue=<rgba 0 6.5002 10 1>}.RGB
    B=VectorParameter() {parameterName=Color 2, defaultValue=<rgba 2 0.0969 0.2562 1>}.RGB
    C=VectorParameter() {defaultValue=<rgba 25 10.3421 0 1>}.RGB
    Alpha=Power(Base=@n27, Exp=@n28)
   ) {materialFunction=&3ColorBlend}.Result) {materialFunction=&MF_FX_EmissiveAdjustments}
  Multiply()
  n29 = MaterialFunctionCall(UVs=TextureCoordinate() {uTiling=5}, "Rotation Angle (0-1)"=ScalarParameter() {defaultValue=0.25, parameterName="Rotation Angle (0-1) (S)"}) {materialFunction=&CustomRotator}
  n30 = TextureSample(UVs=@n29.Rotated Values) {texture=/Game/Global/Textures/Default/Masks_and_Gradients/gradient_stasis}
  n31 = TextureCoordinate()
  n32 = Multiply(A=ComponentMask(None=@n31) {r=true}, B=OneMinus(None=@n31))
  n33 = ScalarParameter() {defaultValue=5, parameterName="Exponent (S)"}
  n34 = ScalarParameter() {defaultValue=0.1, parameterName=BaseReflectFractionIn}
  n35 = MaterialFunctionCall(Rotation Axis=@n11, Rotation Angle=@n13, Normals Pivot=@n14) {materialFunction=&FixRotateAboutAxisNormals}
  n36 = Multiply(A=@n26, B=@n26)
  n37 = ScalarParameter() {defaultValue=0.05, parameterName=OpacityMult}
  n38 = MaterialFunctionCall(Value_Morning=@n37, Value_Day=@n37, Value_Evening=@n37, Value_Night=ScalarParameter() {defaultValue=0.06, parameterName=OpacityMult_Night}) {materialFunction=&Scalar_Time_of_Day}
  n39 = Multiply(
   A=@n25
   B=Multiply(A=Fresnel(ExponentIn=@n33, BaseReflectFractionIn=@n34, Normal=@n35.New Tangent Space Vertex Normal), B=Multiply(A=@n36, B=@n38.Result))
  )
  n40 = ScalarParameter() {defaultValue=100, parameterName=FadeDistance}
  Multiply(A=OneMinus(None=ComponentMask(None=@n30.RGB) {g=true}), B=Multiply(A=Multiply(A=@n32, B=@n39), B=DepthFade(FadeDistance=@n40)))
  TextureSample(UVs=@n29.Rotated Values) {samplerSource=SSM_Wrap_WorldGroupSettings, texture=/Game/Global/Textures/Default/Masks_and_Gradients/T_Vert_Gradient_Sections}
 ]
}
