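# Physics asset for the Star Wand pickaxe, giving its dangling loops, ribbons and ring bodies for secondary motion.
# The handle body combines a shaft capsule and a head sphere; capsules cover loops and ribbons, and a box covers the ring.
# All five dyn bones constrain directly to the handle rather than chaining.
# Loops and ring get even angular limits, while ribbons are tightly restricted, with ribbon_02 swinging mainly on one axis.
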PhysicsAsset /BRCosmetics/Weapons/FORT_Melee/Pickaxe_StarWand/Meshes/Star_Wand_Axe_PhysicsAsset
bodies {
 Melee_TwoHanded_Handle [{shapeType=Sphere, radius=15, center=<0 0 74.5718>}, {shapeType=Capsule, length=86.4268, radius=4.798, center=<1.49 -1.4846 8.196>}]
 dyn_loop_l [{shapeType=Capsule, length=3.3961, radius=2.0571, center=<2.2856 -3.686 -0.8352>, rotation=<rot 0 -152.8813 -73.364>}]
 dyn_loop_r [{shapeType=Capsule, length=5.2807, radius=1.6244, center=<-0.8163 5.065 -0.3819>, rotation=<rot 0 6.6608 -85.6792>}]
 dyn_ribbon_01 [{shapeType=Capsule, length=6.8399, radius=5.065, center=<0 0.2935 8.7322>}]
 dyn_ribbon_02 [{shapeType=Capsule, length=2.4828, radius=1.75, center=<0.0479 -3.5502 0.0715>, rotation=<rot 87.8027 61.6717 151.6914>}]
 dyn_ring_01 [{shapeType=Box, extentX=11.0813, extentY=2.2258, extentZ=13.0108, center=<1.0951 0.2958 -0.465>, rotation=<rot -19.5352 -57.9359 -77.1188>}]
}
constraints [
 {bone1Name=dyn_loop_r, bone2Name=Melee_TwoHanded_Handle, swing1LimitDegrees=45, swing2LimitDegrees=45, twistLimitDegrees=45}
 {bone1Name=dyn_loop_l, bone2Name=Melee_TwoHanded_Handle, swing1LimitDegrees=45, swing2LimitDegrees=45, twistLimitDegrees=45}
 {bone1Name=dyn_ribbon_02, bone2Name=Melee_TwoHanded_Handle, swing1LimitDegrees=80, swing2LimitDegrees=0.01, twistLimitDegrees=2}
 {bone1Name=dyn_ribbon_01, bone2Name=Melee_TwoHanded_Handle, swing1LimitDegrees=0.01, swing2LimitDegrees=4, twistLimitDegrees=8}
 {bone1Name=dyn_ring_01, bone2Name=Melee_TwoHanded_Handle, swing1LimitDegrees=20, swing2LimitDegrees=20, twistLimitDegrees=20}
]
