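# Physics asset for the Star Wand pickaxe, simulating the ribbon bow on its two-handed melee handle.
# The handle body uses a large sphere and a long capsule, while the bow has a small sphere master body and four short capsules.
# The master joint to the handle swings and twists freely.
# The four segments chain from the master with tight angular limits and almost no second-axis swing, so the ribbon bends only slightly.
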PhysicsAsset /BRCosmetics/Weapons/FORT_Melee/Pickaxe_StarWand/Meshes/Star_Wand_Axe_Ribbon_Physics
bodies {
 Melee_TwoHanded_Handle [{shapeType=Sphere, radius=18.75, center=<0 0 76.9782>}, {shapeType=Capsule, length=86.8603, radius=6.8006, center=<1.3355 -1.7006 -1.5365>}]
 dyn_bow_01 [{shapeType=Capsule, length=1.2374, radius=0.9045, center=<0.9146 -0.4818 0.0719>, rotation=<rot 90 -14.0363 35.9606>}]
 dyn_bow_02 [{shapeType=Capsule, length=0.9821, radius=1, center=<1.2983 0.0931 -0.0918>, rotation=<rot 90 0 0>}]
 dyn_bow_03 [{shapeType=Capsule, length=0.5911, radius=1.1615, center=<1.5731 0.0572 -0.06>, rotation=<rot 90 0 0>}]
 dyn_bow_04 [{shapeType=Capsule, length=0.51, radius=1.25, center=<1.617 -0.0745 0.1016>, rotation=<rot 0 -91.5009 -89.155>}]
 dyn_bow_mstr [{shapeType=Sphere, radius=1.7349, center=<-0.1565 0.0674 -0.0253>}]
}
constraints [
 {bone1Name=dyn_bow_01, bone2Name=dyn_bow_mstr, swing1LimitDegrees=5, swing2LimitDegrees=0.01, twistLimitDegrees=5}
 {bone1Name=dyn_bow_02, bone2Name=dyn_bow_01, swing1LimitDegrees=5, swing2LimitDegrees=0.01, twistLimitDegrees=5}
 {bone1Name=dyn_bow_03, bone2Name=dyn_bow_02, swing1LimitDegrees=5, swing2LimitDegrees=0.01, twistLimitDegrees=5}
 {
  bone1Name=dyn_bow_mstr
  bone2Name=Melee_TwoHanded_Handle
  swing1LimitDegrees=45
  swing1Motion=Free
  swing2LimitDegrees=45
  swing2Motion=Free
  twistLimitDegrees=45
  twistMotion=Free
 }
 {bone1Name=dyn_bow_04, bone2Name=dyn_bow_03, swing1LimitDegrees=5, swing2LimitDegrees=0.01, twistLimitDegrees=5}
]
